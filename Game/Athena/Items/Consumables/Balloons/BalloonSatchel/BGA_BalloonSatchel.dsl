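; Balloon satchel gameplay actor that manages the balloons on its owning player pawn and lets the player float.
; Server listens for inflate, deflate, shot and remove-all gameplay events to spawn or remove balloons, switching ballooning movement on or off.
; Vehicle tags, skydiving or unsupported movement release balloons; a timer pops them at the height ceiling, and the actor self-destroys when none remain.
; Clients play floating audio chosen by team affiliation.
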(event EventBeginPlay
  (|CallParentFunction)
  (bind _asfort_player_pawn (Utilities|Casting|CastToFortPlayerPawn (Actor|GetOwner)))
  (bind _output_get (Variables|Default|SetOwningPlayerPawn _asfort_player_pawn))
  (Variables|Default|SetOwningMesh (Variables|/Script/Engine.Character|GetMesh _output_get))
  (CallFunction|SetUpEventResponses)
  (Networking|SwitchHasAuthority
    (:Remote
      (Variables|Default|SetOwnerIsLocal? (Pawn|IsLocallyControlled (Variables|Default|GetOwningPlayerPawn)))
      (bind _output_get_2 (Variables|Default|SetOwnerIsHostileToLocal? (Utilities|Enum|Equal(Enum) (Team|GetTeamAffiliationtoLocalPlayer (Team|GetActorTeam (Variables|Default|GetOwningPlayerPawn))) "Hostile")))
      (Audio|Components|Audio|SetSound (Variables|Default|GetBalloon_FloatingLoop_Cue) (select _output_get_2 (Variables|Default|GetHostileFloatingLoop) (Variables|Default|GetFriendlyFloatingLoop))))))

(event Custom|Set%20Up%20Event%20Responses
  (bind _owningplayerpawn (Variables|Default|GetOwningPlayerPawn))
  (Character|BindEventtoMovementModeChangedDelegate _owningplayerpawn (EventDispatchers|CreateEvent :!event On%20Movement%20Mode%20Changed))
  (EventDispatchers|BindEventtoOnBallooningGravityCeilingChanged _owningplayerpawn (EventDispatchers|CreateEvent :!event Custom|Event_OnBallooningCeilingChanged))
  (Networking|SwitchHasAuthority
    (:Authority
      (bind (_asynctaskproxy _payload) (Ability|Async|WaitGameplayEventToActor _owningplayerpawn "(TagName=\"Event.Athena.Balloons.Inflate\")")
        (:then
          (bind (_asynctaskproxy_2 _payload_2) (Ability|Async|WaitGameplayEventToActor _owningplayerpawn "(TagName=\"Event.Athena.Balloons.Deflate\")")
            (:then
              (Ability|Async|WaitGameplayEventToActor _owningplayerpawn "(TagName=\"Event.Athena.Balloons.RemoveAll\")"
                (:then
                  (bind (_asynctaskproxy_3 _payload_3) (Ability|Async|WaitGameplayEventToActor _owningplayerpawn "(TagName=\"Event.Athena.Balloons.Shot\")")
                    (:then
                      (Ability|Tasks|WaitGameplayTagAddToActor _owningplayerpawn "(TagName=\"Gameplay.Action.Player.AthenaInVehicle\")"
                        (:then
                          (EventDispatchers|BindEventtoOnBeginSkydivingEvent _owningplayerpawn (EventDispatchers|CreateEvent :!event ReleaseAllBalloons)))
                        (:Added
                          (CallFunction|OnPlayerEnteredVehicle))))
                    (:EventReceived
                      (bind _asbga_athena_balloons (Utilities|Casting|CastToBGA_Athena_Balloons (.Target (Utilities|Struct|BreakGameplayEventData _payload_3))))
                      (CallFunction|BalloonShot (.Instigator (Utilities|Struct|BreakGameplayEventData _payload_3)) _asbga_athena_balloons))))
                (:EventReceived
                  (CallFunction|DeflateBalloon true))))
            (:EventReceived
              (for _ (range 1 (+ (Math|Integer|Max(Integer) (Math|Float|Round (.EventMagnitude (Utilities|Struct|BreakGameplayEventData _payload_2))) 1) 1))
                (CallFunction|DeflateBalloon false)
                (Utilities|FlowControl|Delay 0.2)))))
        (:EventReceived
          (for _ (range 1 (+ (Math|Integer|Max(Integer) (Math|Float|Round (.EventMagnitude (Utilities|Struct|BreakGameplayEventData _payload))) 1) 1))
            (CallFunction|InflateBalloon)
            (Utilities|FlowControl|Delay 0.2)))))))

(event Custom|Event_OnBallooningCeilingChanged
  (params (CeilingRelation :type Integer))
  (Math|Integer|CompareInt CeilingRelation
    (:>
      _)
    (:==
      _)
    (:<
      (Utilities|FlowControl|Delay (Variables|Default|GetBelowCeilingHeightBuffer))
      (bind _asfort_movement_comp_character (Utilities|Casting|CastToFortMovementComp_Character (Pawn|GetMovementComponent (Variables|Default|GetOwningPlayerPawn))))
      (if (== (Variables|/Script/FortniteGame.FortMovementComp_Character|GetGravityCeilingRelation _asfort_movement_comp_character) -1)
        (Networking|SwitchHasAuthority
          (:Authority
            (Utilities|Time|ClearandInvalidateTimerbyHandle (Variables|Default|GetAtMaxHeightTimerHandle)))
          (:Remote
            (Audio|Components|Audio|FadeOut (Variables|Default|GetBalloon_MaxAltitude_Cue) 0.5 0.0 "Linear" :!class /Script/Engine.AudioComponent)
            (Utilities|FlowControl|DoOnce false
              (:Completed
                (Audio|Components|Audio|FadeIn (Variables|Default|GetBalloon_MaxAltitude_Cue) 0.5 1.0 0.0 "Linear" :!class /Script/Engine.AudioComponent))))))))
  (Networking|SwitchHasAuthority
    (:Authority
      (if (not (Utilities|Time|IsValidTimerHandle (Variables|Default|GetAtMaxHeightTimerHandle)))
        (bind _returnvalue (Utilities|Time|SetTimerbyEvent (EventDispatchers|CreateEvent :!event PopBalloon) (Variables|Default|GetAboveCeilingHeightPopRate) true))
        (Variables|Default|SetAtMaxHeightTimerHandle _returnvalue)))
    (:Remote)))

(event EventEndPlay
  (params (EndPlayReason :type EEndPlayReason%20Enum :default "Destroyed"))
  (|CallParentFunction EndPlayReason)
  (Default|ReleaseAllBalloons))

(fn BalloonShot
  (params (Instigator :type Actor%20Object%20Reference) (Balloon :type BGA%20Athena%20Balloons%20Object%20Reference))
  (Utilities|Array|RemoveItem (Variables|Default|GetBalloonList) Balloon)
  (Balloon|SetNumActiveBalloons (Variables|Default|GetOwningPlayerPawn) (- (Balloon|GetNumActiveBalloons (Variables|Default|GetOwningPlayerPawn)) 1))
  (CallFunction|ReorderBalloons)
  (CallFunction|UpdateMovement))

(fn CheckAutoDestroy
  (if (not (> (Balloon|GetNumActiveBalloons (Variables|Default|GetOwningPlayerPawn)) 0))
    (Actor|DestroyActor)))

(fn DeflateBalloon
  (params (DeflateAll? :type Boolean) (Pop? :type Boolean))
  (bind _balloonlist (Variables|Default|GetBalloonList))
  (bind _returnvalue (Utilities|Array|LastIndex _balloonlist))
  (bind _owningplayerpawn (Variables|Default|GetOwningPlayerPawn))
  (bind _returnvalue_2 (Balloon|GetNumActiveBalloons _owningplayerpawn))
  (bind _output (Utilities|Array|Get(acopy) _balloonlist _returnvalue))
  (Utilities|FlowControl|Sequence
    (:then_0
      (if (> _returnvalue_2 0)
        (Balloon|SetNumActiveBalloons (Variables|Default|GetOwningPlayerPawn) (Math|Integer|SelectInt 0 (- _returnvalue_2 1) DeflateAll?))
        (if (Variables|GetDeflateAll?)
          (while (not (Utilities|Array|IsEmpty (Variables|Default|GetBalloonList)))
            (if (Variables|GetPop?)
              (Class|BGAAthenaBalloons|PopatMaxHeight _output)
              (Actor|SetLifeSpan _output (LocalVariables|GetLifespanAfterPop))
              (else
                (Class|BGAAthenaBalloons|Released _output)))
            (Utilities|Array|RemoveIndex (Variables|Default|GetBalloonList) _returnvalue)
            (if (not (Variables|GetDeflateAll?))
              (CallFunction|ReorderBalloons)
              (CallFunction|UpdateMovement)
              (if (not (Variables|GetPop?))
                (Ability|ExecuteGameplayCueReplicated (Variables|Default|GetGameplayCueDeflate)))))
          (CallFunction|UpdateMovement)
          (if (not (Variables|GetPop?))
            (Ability|ExecuteGameplayCueReplicated (Variables|Default|GetGameplayCueDeflateMulti))))))
    (:then_1
      (if (not (> (Balloon|GetNumActiveBalloons (Variables|Default|GetOwningPlayerPawn)) 0))
        (Utilities|Time|SetTimerbyEvent (EventDispatchers|CreateEvent :!event CheckAutoDestroy) (Variables|Default|GetAutoDestroyDelay))
        (GameplayEffects|RemoveActiveGameplayEffect (Ability|GetAbilitySystemComponent (Variables|Default|GetOwningPlayerPawn)) (Variables|Default|GetBalloonQuestGEHandle) -1)))))

(fn InflateBalloon
  (if (Networking|IsServer)
    (if (< (Balloon|GetNumActiveBalloons (Variables|Default|GetOwningPlayerPawn)) (Variables|Default|GetNumMaxBalloons))
      (bind _returnvalue (Game|SpawnActorfromClass "/Game/Athena/Items/Consumables/Balloons/BalloonSatchel/BGA_Athena_Balloons_FriendlyFire.BGA_Athena_Balloons_FriendlyFire_C" (Math|Transform|MakeTransform (.location (Transformation|GetSocketTransform (Variables|Default|GetBalloon_SKBudgeted) (LocalVariables|GetBalloonAttachSocketName) "RTS_World")) (Variables|Default|GetRotationOffset) (.scale (Transformation|GetSocketTransform (Variables|Default|GetBalloon_SKBudgeted) (LocalVariables|GetBalloonAttachSocketName) "RTS_World"))) "AlwaysSpawn" "OverrideRootScale" (Variables|Default|GetOwningPlayerPawn) (Math|Random|RandomIntegerinRange 0 5) (Math|Random|RandomIntegerinRange 0 6) (Utilities|Array|Length (Variables|Default|GetBalloonList)) 0 0))
      (Utilities|Array|Add (Variables|Default|GetBalloonList) _returnvalue)
      (Transformation|AttachActorToActor _returnvalue self "None" "KeepWorld" "KeepRelative" "KeepRelative" true)
      (Balloon|SetNumActiveBalloons (Variables|Default|GetOwningPlayerPawn) (+ (Balloon|GetNumActiveBalloons (Variables|Default|GetOwningPlayerPawn)) 1))
      (CallFunction|UpdateMovement)
      (Ability|ExecuteGameplayCueReplicated (Variables|Default|GetGameplayCueInflate))
      (bind _returnvalue_2 (GameplayEffects|ApplyGameplayEffectToSelf :self (Ability|GetAbilitySystemComponent (Variables|Default|GetOwningPlayerPawn)) :GameplayEffectClass (Variables|Default|GetBalloonQuestGEClass) :Level 1.0))
      (Variables|Default|SetBalloonQuestGEHandle _returnvalue_2))))

(fn OnMovementModeChanged
  (params (Character :type Character%20Object%20Reference) (PrevMovementMode :type EMovementMode%20Enum :default "MOVE_None") (PreviousCustomMode :type Byte))
  (bind _balloon_floatingloop_cue (Variables|Default|GetBalloon_FloatingLoop_Cue))
  (bind _balloon_starttorise_cue (Variables|Default|GetBalloon_StartToRise_Cue))
  (Networking|SwitchHasAuthority
    (:Authority
      (if (Utilities|Array|ContainsItem (Variables|Default|GetUnsupportedCustomMovementModes) (Utilities|Enum|BytetoEnumEFortCustomMovement (Variables|/Script/Engine.CharacterMovementComponent|GetCustomMovementMode (Variables|/Script/Engine.Character|GetCharacterMovement (Variables|Default|GetOwningPlayerPawn)))))
        (Default|ReleaseAllBalloons)
        (else
          (CallFunction|UpdateMovement))))
    (:Remote
      (if (Utilities|Enum|Equal(Enum) (Utilities|Enum|BytetoEnumEFortCustomMovement (Variables|/Script/Engine.CharacterMovementComponent|GetCustomMovementMode (Variables|/Script/Engine.Character|GetCharacterMovement (Variables|GetCharacter)))) "Ballooning")
        (Audio|Components|Audio|FadeIn _balloon_floatingloop_cue 0.25 1.0 0.0 "Linear" :!class /Script/Engine.AudioComponent)
        (if (Variables|Default|GetOwnerIsHostileToLocal?)
          (Components|Activation|Activate _balloon_starttorise_cue false))
        (else
          (Audio|Components|Audio|FadeOut _balloon_floatingloop_cue 0.25 0.0 "Linear" :!class /Script/Engine.AudioComponent)
          (Audio|Components|Audio|FadeOut _balloon_starttorise_cue 0.25 0.0 "Linear" :!class /Script/Engine.AudioComponent))))))

(fn OnPlayerEnteredVehicle
  (bind _owningplayerpawn (Variables|Default|GetOwningPlayerPawn))
  (bind _returnvalue (Vehicle|GetVehicleInterface _owningplayerpawn))
  (bind _returnvalue_2 (Utilities|ToObject(Interface) _returnvalue))
  (Utilities|IsValid _returnvalue_2
    (:Is%20Valid
      (bind _returnvalue_3 (Class|FortVehicleInterface|GetGameplayTags(Message) _returnvalue_2))
      (if (GameplayTags|HasTag _returnvalue_3 "(TagName=\"Athena.Vehicle.OnEntry.ReleaseBalloons\")")
        (CallFunction|DeflateBalloon true)))))

(fn PopBalloon
  (CallFunction|DeflateBalloon false true))

(fn ReleaseAllBalloons
  (CallFunction|DeflateBalloon true))

(fn ReorderBalloons
  (for (_array_element _array_index) (Variables|Default|GetBalloonList)
    (Variables|/Script/FortniteGame.BuildingGameplayActorBalloon|SetBalloonLocationSelection _array_index _array_element)))

(fn UpdateMovement
  (if (> (Balloon|GetNumActiveBalloons (Variables|Default|GetOwningPlayerPawn)) 0)
    (if (AI|Components|NavMovement|IsFalling (Variables|/Script/Engine.Character|GetCharacterMovement (Variables|Default|GetOwningPlayerPawn)))
      _
      (elif (and (Pawn|Components|CharacterMovement|IsWalking (Variables|/Script/Engine.Character|GetCharacterMovement (Variables|Default|GetOwningPlayerPawn))) (>= (Balloon|GetNumActiveBalloons (Variables|Default|GetOwningPlayerPawn)) (Variables|Default|GetNumBalloonsRequiredForFlight)))
        _))
    (Pawn|Components|CharacterMovement|SetMovementMode (Variables|/Script/Engine.Character|GetCharacterMovement (Variables|Default|GetOwningPlayerPawn)) "MOVE_Custom" (Variables|Default|GetBallooningCustomMovementMode))
    (Character|UnCrouch (Variables|Default|GetOwningPlayerPawn))
    (elif (Utilities|Enum|Equal(Enum) (Utilities|Enum|BytetoEnumEFortCustomMovement (Variables|/Script/Engine.CharacterMovementComponent|GetCustomMovementMode (Variables|/Script/Engine.Character|GetCharacterMovement (Variables|Default|GetOwningPlayerPawn)))) "Ballooning")
      (Pawn|Components|CharacterMovement|SetMovementMode (Variables|/Script/Engine.Character|GetCharacterMovement (Variables|Default|GetOwningPlayerPawn)) "MOVE_Falling" 0))))

(fn ConstructionScript)
