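; Balloon satchel gameplay ability that signals a balloon deflate on the activating player pawn.
; On activation it commits, then checks that the pawn has a valid balloon rope and casts it to the balloon satchel actor.
; It then sends the Event.Athena.Balloons.Deflate gameplay event to the pawn, passing the ability level as event magnitude, and ends.
; An invalid rope leaves that branch unhandled, so EndAbility is not called there.
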(event Ability|EventActivateAbility
  (Ability|CommitAbility)
  (bind _asfort_player_pawn (Utilities|Casting|CastToFortPlayerPawn (Ability|GetActivatingPawn)))
  (Utilities|IsValid (Variables|/Script/FortniteGame.FortPlayerPawn|GetBalloonRope _asfort_player_pawn)
    (:Is%20Valid
      (Utilities|Casting|CastToBGA_BalloonSatchel (Variables|/Script/FortniteGame.FortPlayerPawn|GetBalloonRope _asfort_player_pawn))
      (Ability|SendGameplayEventtoActor _asfort_player_pawn "(TagName=\"Event.Athena.Balloons.Deflate\")" (Utilities|Struct|MakeGameplayEventData "(TagName=\"\")" 0 0 0 0 "()" "(GameplayTags=)" "(GameplayTags=)" (Math|Conversions|ToFloat(Integer) (Ability|GetAbilityLevel))))
      (Ability|EndAbility))))

(event Ability|EventOnEndAbility
  (params (bWasCancelled :type Boolean)))

(event Ability|EventFailedToActivatePassiveAbility
  (params (ActorInfo :type Gameplay%20Ability%20Actor%20Info%20Structure))
  (|CallParentFunction))
