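; Gameplay ability for the Balloon Satchel consumable that triggers balloon inflation on the activating player pawn.
; On activation it commits, then checks the pawn's BalloonRope; if none exists, it spawns a BGA_BalloonSatchel, stores it as the rope and attaches it to the mesh's backpack socket.
; It sends the Event.Athena.Balloons.Inflate gameplay event to the pawn with the ability level as magnitude, then ends immediately.
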(event Ability|EventActivateAbility
  (Ability|CommitAbility)
  (bind _asfort_player_pawn (Utilities|Casting|CastToFortPlayerPawn (Ability|GetActivatingPawn)))
  (Utilities|IsValid (Variables|/Script/FortniteGame.FortPlayerPawn|GetBalloonRope _asfort_player_pawn)
    (:Is%20Valid
      (Utilities|Casting|CastToBGA_BalloonSatchel (Variables|/Script/FortniteGame.FortPlayerPawn|GetBalloonRope _asfort_player_pawn)))
    (:Is%20Not%20Valid
      (bind _returnvalue (Game|SpawnActorfromClass "/Game/Athena/Items/Consumables/Balloons/BalloonSatchel/BGA_BalloonSatchel.BGA_BalloonSatchel_C" (Transformation|GetActorTransform _asfort_player_pawn) "AlwaysSpawn" "OverrideRootScale" _asfort_player_pawn 0 _asfort_player_pawn))
      (Balloon|SetBalloonRope _asfort_player_pawn _returnvalue)
      (Transformation|AttachActorToComponent _returnvalue (Variables|/Script/Engine.Character|GetMesh _asfort_player_pawn) "backpack" "SnapToTarget" "KeepWorld" "KeepWorld" true)))
  (Ability|SendGameplayEventtoActor _asfort_player_pawn "(TagName=\"Event.Athena.Balloons.Inflate\")" (Utilities|Struct|MakeGameplayEventData "(TagName=\"\")" 0 0 0 0 "()" "(GameplayTags=)" "(GameplayTags=)" (Math|Conversions|ToFloat(Integer) (Ability|GetAbilityLevel))))
  (Ability|EndAbility))

(event Ability|EventOnEndAbility
  (params (bWasCancelled :type Boolean)))

(event Ability|EventFailedToActivatePassiveAbility
  (params (ActorInfo :type Gameplay%20Ability%20Actor%20Info%20Structure))
  (|CallParentFunction))
